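# Material function for the Street Cuddles body's eyes, rendering an animated LED screen with glitch effects.
# Hashed rows periodically shift the LED texture horizontally, and its rounded blue channel gates emission alongside a hashed flicker term.
# Tiled scanlines, modulated by sine- and tangent-driven moving bars, are tinted and added to the glow.
# The result overrides emissive, base color, specular, roughness and metallic on the incoming material attributes.
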MaterialFunction /BRCosmetics/Characters/Player/Male/Medium/Bodies/M_MED_Street_Cuddles/Materials/MF_StreetCuddlesEyes
graph {nodes [
  n1 = Time()
  Add(A=@n1)
  n2 = TextureCoordinate()
  n3 = AppendVector(
   A=ScalarParameter() {group=StreetCuddles, parameterName=SC_Pixel_VerticalOffset, sliderMax=1}
   B=ScalarParameter() {group=StreetCuddles, parameterName=SC_Pixel_HorizontalOffset, sliderMax=1}
  )
  n4 = Ceil(None=Multiply(A=Add(A=@n2, B=@n3), B=ScalarParameter() {defaultValue=50, group=StreetCuddles, parameterName=SC_GlitchLine_Tiling}))
  ComponentMask(None=@n4) {g=true}
  Constant() {r=0.05}
  Constant() {r=10}
  n5 = TextureSampleParameter2D() {
   group=StreetCuddles
   parameterName=SC_LED_Texture
   texture=/BRCosmetics/Characters/Player/Male/Medium/Bodies/M_MED_Street_Cuddles/Textures/M_MED_Street_Cuddles_faceAccNoHud_D
  }
  n6 = SphereMask(A=@n5.RGB, B=Constant() {r=1}, Radius=Constant())
  n7 = Multiply(A=TextureCoordinate(), B=ScalarParameter() {defaultValue=50, group=StreetCuddles, parameterName=SC_GlitchLine_Tiling})
  n8 = MaterialFunctionCall(Vec2=ComponentMask(None=Ceil(None=@n7)) {g=true}) {materialFunction=&MF_Hash}
  n9 = Multiply(A=@n1, B=ScalarParameter() {defaultValue=0.2, group=StreetCuddles, parameterName=SC_GlitchFrequency}) {constB=0.05}
  n10 = ScalarParameter() {defaultValue=0.8, group=StreetCuddles, parameterName=SC_GlitchThreshold_Min}
  n11 = ScalarParameter() {defaultValue=0.9, group=StreetCuddles, parameterName=SC_GlitchThreshold_Max}
  n12 = Time()
  n13 = Round(None=ConstantBiasScale(None=MaterialFunctionCall(Float=@n12) {materialFunction=&MF_Hash}.Float))
  n14 = Ceil(None=Subtract(A=Frac(None=@n9), B=LinearInterpolate(A=@n10, B=@n11, Alpha=@n13)))
  n15 = Multiply(
   A=Multiply(A=ComponentMask(None=@n8.Vec2) {r=true}, B=Saturate(None=@n14))
   B=ScalarParameter() {defaultValue=0.04, group=StreetCuddles, parameterName=SC_GlitchScaleMult}
  )
  n16 = MaterialFunctionCall(Float2=TextureCoordinate()) {materialFunction=&BreakOutFloat2Components}
  n17 = TextureSampleParameter2D(UVs=AppendVector(A=Add(A=@n15, B=@n16.R), B=@n16.G)) {
   group=StreetCuddles
   parameterName=SC_LED_Texture
   texture=&M_MED_Street_Cuddles_faceAccNoHud_D
  }
  n18 = Round(None=@n17.B)
  n19 = Multiply(A=Time(), B=ScalarParameter() {defaultValue=2, group=StreetCuddles, parameterName=SC_FlickerFrequency})
  n20 = Add(
   A=@n18
   B=ComponentMask(None=MaterialFunctionCall(Seed=@n19, Vec2=@n4) {materialFunction=&MF_Hash}.Vec2) {r=true}
  )
  n21 = LinearInterpolate(B=@n17.RGB, Alpha=LinearInterpolate(B=Saturate(None=@n20), Alpha=@n18))
  n22 = Divide(B=ScalarParameter() {defaultValue=0.04, group=StreetCuddles, parameterName=SC_LineThickness})
  n23 = Multiply(
   A=MaterialFunctionCall(Float2=TextureCoordinate()) {materialFunction=&BreakOutFloat2Components}.G
   B=@n22
  )
  n24 = Divide(A=@n22, B=ScalarParameter() {defaultValue=40, group=StreetCuddles, parameterName=SC_LineTiling})
  n25 = Saturate(None=SphereMask(A=Fmod(A=@n23, B=@n24), B=Constant() {r=0.25}, Radius=Constant() {r=0.1}, Hardness=Constant()))
  n26 = Multiply(A=Time(), B=ScalarParameter() {defaultValue=0.2, group=StreetCuddles, parameterName=SC_MovingBar2_Speed})
  n27 = SphereMask(
   A=MaterialFunctionCall(Float2=TextureCoordinate()) {materialFunction=&BreakOutFloat2Components}.G
   B=ConstantBiasScale(None=Tangent!(None=@n26)) {bias=0.5, scale=0.25}
   Radius=ScalarParameter() {defaultValue=0.01, group=StreetCuddles, parameterName=SC_MovingBar2_Thickness}
   Hardness=Constant()
  )
  n28 = ScalarParameter() {defaultValue=1, group=StreetCuddles, parameterName=SC_MovingBar2_Opacity}
  n29 = Multiply(A=Time(), B=ScalarParameter() {defaultValue=0.1, group=StreetCuddles, parameterName=SC_MovingBar1_Speed})
  n30 = SphereMask(
   A=MaterialFunctionCall(Float2=TextureCoordinate()) {materialFunction=&BreakOutFloat2Components}.G
   B=ConstantBiasScale(None=Sine(None=@n29)) {bias=0.5, scale=0.2}
   Radius=ScalarParameter() {defaultValue=0.05, group=StreetCuddles, parameterName=SC_MovingBar1_Thickness}
   Hardness=Constant()
  )
  n31 = ScalarParameter() {defaultValue=0.1, group=StreetCuddles, parameterName=SC_MovingBar1_Opacity}
  n32 = Add(A=Add(A=Multiply(A=@n27, B=@n28) {constB=0.1}, B=Multiply(A=@n30, B=@n31) {constB=0.1})) {constA=0.8, constB=0.05}
  n33 = Multiply(
   A=LinearInterpolate(B=@n25, Alpha=Saturate(None=@n32))
   B=VectorParameter() {group=StreetCuddles, parameterName=SC_MovingBars_Color, defaultValue=<rgba 1 0 0.4486 1>}.RGB
  )
  n34 = Multiply(
   A=Add(A=Saturate(None=@n6), B=Add(A=@n21, B=@n33))
   B=ScalarParameter() {defaultValue=0.5, group=StreetCuddles, parameterName=SC_EmissiveMult}
  ) {constB=0.5}
  n35 = SetMaterialAttributes!(
   MaterialAttributes=FunctionInput() {inputName=MaterialAttributes, inputType=FunctionInput_MaterialAttributes}
   EmissiveColor=@n34
   Specular=Constant() {r=1}
   BaseColor=@n17.RGB
   Roughness=ScalarParameter() {defaultValue=0.25, group=StreetCuddles, parameterName=SC_Roughness}
   Metallic=Constant()
  ) {attributeSetTypes [
    #1
    #2
    #3
    #4
    #5
   ]}
  FunctionOutput(None=@n35)
  Multiply(A=@n34, B=Add(A=@n25) {constB=0.5})
  n36 = Frac(None=@n19)
  Saturate(None=Ceil(None=Subtract(A=@n36)))
 ]}
