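# Master material for the eyes of the Street Cuddles male medium player character, a Default Lit skeletal mesh material using material attributes.
# The shared character PreFX function feeds its attributes through MF_StreetCuddlesEyes and into the shared PostFX function, sandwiching the eye-specific look between common character stages.
# PreFX's AO output also drives PostFX's AO, Distance Blend and Fresnel inputs.
# Usage flags support morph targets and mesh deformers.
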Material /BRCosmetics/Characters/Player/Male/Medium/Bodies/M_MED_Street_Cuddles/Materials/M_MED_StreetCuddles_Eye_MASTER
graph {
 properties {
  bUseMaterialAttributes=true
  bUsedWithMeshDeformer=true
  bUsedWithMorphTargets=true
  bUsedWithSkeletalMesh=true
  pixelDepthOffsetMode=PDOM_Legacy
  usedShadingModels=Default Lit
 }
 nodes [
  n1 = MaterialFunctionCall() {materialFunction=&BaseFN_CharacterFunction_PreFX}
  MaterialFunctionCall(
   AO=@n1.AO
   Distance Blend=@n1.AO
   Fresnel=@n1.AO
   Material=MaterialFunctionCall!(MaterialAttributes=@n1.Result) {materialFunction=&/MF_StreetCuddlesEyes}.Result
  ) {materialFunction=&BaseFN_CharacterFunction_PostFX}
 ]
}
