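# Physics asset for the suspenders on the crouching Street Cuddles male body, with collision shapes on pelvis, thighs and calves.
# Four strap chains (left/right, front/back) hang from MSTR bodies linked to the pelvis, using 30-degree swing and twist limits between segments.
# Free-rotation constraints cross-link front and back tips on each side, closing them into loops.
# Straps use capsules, except boxes at the right-side ends; calf_l has no constraint.
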PhysicsAsset /BRCosmetics/Characters/Player/Male/Medium/Bodies/M_MED_Street_Cuddles/Meshes/M_MED_Street_Cuddles_Crouching_Suspenders_Physics
bodies {
 calf_l [{shapeType=Capsule, length=35, radius=9.0223, shapeName=calf_l_capsule, center=<-19.0238 -2.731 -0.3306>, rotation=<rot 90 0 0>}]
 calf_r [{shapeType=Capsule, length=35, radius=9.0223, shapeName=calf_r_capsule, center=<19.0238 2.731 0.3306>, rotation=<rot 90 180 0>}]
 dyn_l_backStrap_01 [{shapeType=Capsule, length=9.6161, radius=2.1691, shapeName=dyn_l_backStrap_01_capsule, center=<3.2617 0 -0.0001>, rotation=<rot 0 0 90>}]
 dyn_l_backStrap_02 [{shapeType=Capsule, length=9.6161, radius=2.1691, shapeName=dyn_l_backStrap_02_capsule, center=<3.2617 0 -0.0001>, rotation=<rot 0 0 90>}]
 dyn_l_backStrap_03 [{shapeType=Capsule, length=9.6161, radius=2.1691, shapeName=dyn_l_backStrap_03_capsule, center=<3.2617 0 -0.0001>, rotation=<rot 0 0 90>}]
 dyn_l_backStrap_04 [{shapeType=Capsule, length=9.6161, radius=2.1691, shapeName=dyn_l_backStrap_04_capsule, center=<3.2617 0 -0.0001>, rotation=<rot 0 0 90>}]
 dyn_l_backStrap_05 [{shapeType=Capsule, length=5, radius=1.5, shapeName=dyn_l_backStrap_05_capsule, center=<2.0713 0 0>, rotation=<rot 0 0 90>}]
 dyn_l_backStrap_06 [{shapeType=Capsule, length=5, radius=1.5, shapeName=dyn_l_backStrap_06_capsule, center=<2.0713 0 0>, rotation=<rot 0 0 90>}]
 dyn_l_backStrap_MSTR [{shapeType=Capsule, length=1, radius=1}]
 dyn_l_frontStrap_01 [{shapeType=Capsule, length=9.474, radius=2.0309, shapeName=dyn_l_frontStrap_01_capsule, center=<3.7325 0 0.0001>, rotation=<rot 0 0 90>}]
 dyn_l_frontStrap_02 [{shapeType=Capsule, length=9.474, radius=2.0309, shapeName=dyn_l_frontStrap_02_capsule, center=<3.7325 0 0.0001>, rotation=<rot 0 0 90>}]
 dyn_l_frontStrap_03 [{shapeType=Capsule, length=9.474, radius=2.0309, shapeName=dyn_l_frontStrap_03_capsule, center=<3.7325 0 0.0001>, rotation=<rot 0 0 90>}]
 dyn_l_frontStrap_04 [{shapeType=Capsule, length=9.474, radius=2.0309, shapeName=dyn_l_frontStrap_04_capsule, center=<3.7325 0 0.0001>, rotation=<rot 0 0 96.0899>}]
 dyn_l_frontStrap_05 [{shapeType=Capsule, length=5, radius=1.5, shapeName=dyn_l_frontStrap_05_capsule, center=<2.0713 0 0>, rotation=<rot 0 0 90>}]
 dyn_l_frontStrap_06 [{shapeType=Capsule, length=5, radius=1.5, shapeName=dyn_l_frontStrap_06_capsule, center=<2.0713 0 0>, rotation=<rot 0 0 90>}]
 dyn_l_frontStrap_07 [{shapeType=Capsule, length=5, radius=1.5, shapeName=dyn_l_frontStrap_07_capsule, center=<2.0713 0 0>, rotation=<rot 0 0 90>}]
 dyn_l_frontStrap_MSTR [{shapeType=Capsule, length=1, radius=1}]
 dyn_r_backStrap_01 [{shapeType=Capsule, length=9.2864, radius=2.6195, shapeName=dyn_r_backStrap_01_capsule, center=<3.7239 0 0>, rotation=<rot 0 0 85.4226>}]
 dyn_r_backStrap_02 [{shapeType=Capsule, length=9.4756, radius=2.4303, shapeName=dyn_r_backStrap_02_capsule, center=<3.7239 0 0>, rotation=<rot 0 0 82.9015>}]
 dyn_r_backStrap_03 [{shapeType=Capsule, length=9.9059, radius=3, shapeName=dyn_r_backStrap_03_capsule, center=<3.7239 0 0>, rotation=<rot 0 0 89.3505>}]
 dyn_r_backStrap_04 [{shapeType=Capsule, length=9.2995, radius=3, shapeName=dyn_r_backStrap_04_capsule, center=<3.7239 0 0>, rotation=<rot 0 0 68.6526>}]
 dyn_r_backStrap_05 [{shapeType=Box, extentX=4.6786, extentY=5.763, extentZ=3.8257, center=<2.3012 0 -0.0001>}]
 dyn_r_backStrap_06 [{shapeType=Box, extentX=4.4711, extentY=5.2793, extentZ=3.8257, center=<1.4363 0 0>}]
 dyn_r_backStrap_MSTR [{shapeType=Capsule, length=1, radius=1}]
 dyn_r_frontStrap_01 [{shapeType=Capsule, length=9.9171, radius=2, shapeName=dyn_r_frontStrap_01_capsule, center=<2.7589 0 0>, rotation=<rot -0.0001 0 117.5103>}]
 dyn_r_frontStrap_02 [{shapeType=Capsule, length=9.9171, radius=2, shapeName=dyn_r_frontStrap_02_capsule, center=<2.7589 0 0>, rotation=<rot 0 0 117.5124>}]
 dyn_r_frontStrap_03 [{shapeType=Capsule, length=9.9171, radius=2, shapeName=dyn_r_frontStrap_03_capsule, center=<2.7589 0 0>, rotation=<rot 0 0 117.5114>}]
 dyn_r_frontStrap_04 [{shapeType=Capsule, length=9.3498, radius=2.5673, shapeName=dyn_r_frontStrap_04_capsule, center=<2.7589 0 0>, rotation=<rot 0 0 117.518>}]
 dyn_r_frontStrap_05 [{shapeType=Box, extentX=4.4711, extentY=5.085, extentZ=3.8257, center=<1.7615 0 0.0001>}]
 dyn_r_frontStrap_06 [{shapeType=Box, extentX=4.4876, extentY=5.763, extentZ=2.3265, center=<1.8712 0 0>}]
 dyn_r_frontStrap_07 [{shapeType=Box, extentX=5.1289, extentY=5.763, extentZ=4.8501, center=<2.4786 0 0.0001>}]
 dyn_r_frontStrap_MSTR [{shapeType=Capsule, length=1, radius=1}]
 pelvis [{shapeType=Capsule, length=11.2747, radius=8.2912, center=<-0.9194 -1.7087 0.0001>}]
 thigh_l [
  {shapeType=Capsule, length=24, radius=8.9, center=<-20.1184 -0.0007 -0.3471>, rotation=<rot 90 -178.1873 0>}
  {shapeType=Capsule, length=22.6193, radius=9.1514, center=<-19.4514 -0.0008 7.8363>, rotation=<rot 86.627 0 180>}
 ]
 thigh_r [
  {shapeType=Capsule, length=30, radius=7, center=<20.1184 0.0007 0.3471>, rotation=<rot 90 90 88.1873>}
  {shapeType=Capsule, length=22.6193, radius=9.1514, center=<19.4514 0.0008 -7.8363>, rotation=<rot 86.627 0 0>}
 ]
 thigh_twist_01_l [{shapeType=Capsule, length=24.8462, radius=8, center=<1.878 -0.0005 -0.3471>, rotation=<rot 90 180 0>}]
 thigh_twist_01_r [{shapeType=Capsule, length=24.8462, radius=7, center=<-1.878 0.0005 0.3471>, rotation=<rot 90 0 0>}]
}
constraints [
 {
  bone1Name=thigh_l
  bone2Name=pelvis
  swing1LimitDegrees=45
  swing1Motion=Free
  swing2LimitDegrees=45
  swing2Motion=Free
  twistLimitDegrees=45
  twistMotion=Free
 }
 {
  bone1Name=thigh_r
  bone2Name=pelvis
  swing1LimitDegrees=45
  swing1Motion=Free
  swing2LimitDegrees=45
  swing2Motion=Free
  twistLimitDegrees=45
  twistMotion=Free
 }
 {bone1Name=dyn_l_backStrap_01, bone2Name=dyn_l_backStrap_MSTR, swing1LimitDegrees=30, swing2LimitDegrees=30, twistLimitDegrees=30}
 {bone1Name=dyn_l_frontStrap_01, bone2Name=dyn_l_frontStrap_MSTR, swing1LimitDegrees=30, swing2LimitDegrees=30, twistLimitDegrees=30}
 {bone1Name=dyn_l_frontStrap_02, bone2Name=dyn_l_frontStrap_01, swing1LimitDegrees=30, swing2LimitDegrees=30, twistLimitDegrees=30}
 {bone1Name=dyn_l_frontStrap_03, bone2Name=dyn_l_frontStrap_02, swing1LimitDegrees=30, swing2LimitDegrees=30, twistLimitDegrees=30}
 {bone1Name=dyn_l_frontStrap_04, bone2Name=dyn_l_frontStrap_03, swing1LimitDegrees=30, swing2LimitDegrees=30, twistLimitDegrees=30}
 {bone1Name=dyn_l_frontStrap_05, bone2Name=dyn_l_frontStrap_04, swing1LimitDegrees=30, swing2LimitDegrees=30, twistLimitDegrees=30}
 {bone1Name=dyn_l_frontStrap_06, bone2Name=dyn_l_frontStrap_05, swing1LimitDegrees=30, swing2LimitDegrees=30, twistLimitDegrees=30}
 {bone1Name=dyn_l_frontStrap_07, bone2Name=dyn_l_frontStrap_06, swing1LimitDegrees=30, swing2LimitDegrees=30, twistLimitDegrees=30}
 {bone1Name=dyn_l_backStrap_02, bone2Name=dyn_l_backStrap_01, swing1LimitDegrees=30, swing2LimitDegrees=30, twistLimitDegrees=30}
 {bone1Name=dyn_l_backStrap_03, bone2Name=dyn_l_backStrap_02, swing1LimitDegrees=30, swing2LimitDegrees=30, twistLimitDegrees=30}
 {bone1Name=dyn_l_backStrap_04, bone2Name=dyn_l_backStrap_03, swing1LimitDegrees=30, swing2LimitDegrees=30, twistLimitDegrees=30}
 {bone1Name=dyn_l_backStrap_05, bone2Name=dyn_l_backStrap_04, swing1LimitDegrees=30, swing2LimitDegrees=30, twistLimitDegrees=30}
 {bone1Name=dyn_l_backStrap_06, bone2Name=dyn_l_backStrap_05, swing1LimitDegrees=30, swing2LimitDegrees=30, twistLimitDegrees=30}
 {bone1Name=dyn_r_frontStrap_01, bone2Name=dyn_r_frontStrap_MSTR, swing1LimitDegrees=30, swing2LimitDegrees=30, twistLimitDegrees=30}
 {bone1Name=dyn_r_frontStrap_02, bone2Name=dyn_r_frontStrap_01, swing1LimitDegrees=30, swing2LimitDegrees=30, twistLimitDegrees=30}
 {bone1Name=dyn_r_frontStrap_03, bone2Name=dyn_r_frontStrap_02, swing1LimitDegrees=30, swing2LimitDegrees=30, twistLimitDegrees=30}
 {bone1Name=dyn_r_frontStrap_04, bone2Name=dyn_r_frontStrap_03, swing1LimitDegrees=30, swing2LimitDegrees=30, twistLimitDegrees=30}
 {bone1Name=dyn_r_frontStrap_05, bone2Name=dyn_r_frontStrap_04, swing1LimitDegrees=30, swing2LimitDegrees=30, twistLimitDegrees=30}
 {bone1Name=dyn_r_frontStrap_06, bone2Name=dyn_r_frontStrap_05, swing1LimitDegrees=30, swing2LimitDegrees=30, twistLimitDegrees=30}
 {bone1Name=dyn_r_frontStrap_07, bone2Name=dyn_r_frontStrap_06, swing1LimitDegrees=30, swing2LimitDegrees=30, twistLimitDegrees=30}
 {bone1Name=dyn_r_backStrap_01, bone2Name=dyn_r_backStrap_MSTR, swing1LimitDegrees=30, swing2LimitDegrees=30, twistLimitDegrees=30}
 {bone1Name=dyn_r_backStrap_02, bone2Name=dyn_r_backStrap_01, swing1LimitDegrees=30, swing2LimitDegrees=30, twistLimitDegrees=30}
 {bone1Name=dyn_r_backStrap_03, bone2Name=dyn_r_backStrap_02, swing1LimitDegrees=30, swing2LimitDegrees=30, twistLimitDegrees=30}
 {bone1Name=dyn_r_backStrap_04, bone2Name=dyn_r_backStrap_03, swing1LimitDegrees=30, swing2LimitDegrees=30, twistLimitDegrees=30}
 {bone1Name=dyn_r_backStrap_05, bone2Name=dyn_r_backStrap_04, swing1LimitDegrees=30, swing2LimitDegrees=30, twistLimitDegrees=30}
 {bone1Name=dyn_r_backStrap_06, bone2Name=dyn_r_backStrap_05, swing1LimitDegrees=30, swing2LimitDegrees=30, twistLimitDegrees=30}
 {
  bone1Name=dyn_l_frontStrap_06
  bone2Name=dyn_l_backStrap_06
  swing1LimitDegrees=45
  swing1Motion=Free
  swing2LimitDegrees=45
  swing2Motion=Free
  twistLimitDegrees=45
  twistMotion=Free
 }
 {
  bone1Name=dyn_l_backStrap_05
  bone2Name=dyn_l_frontStrap_07
  swing1LimitDegrees=45
  swing1Motion=Free
  swing2LimitDegrees=45
  swing2Motion=Free
  twistLimitDegrees=45
  twistMotion=Free
 }
 {
  bone1Name=dyn_r_backStrap_06
  bone2Name=dyn_r_frontStrap_06
  swing1LimitDegrees=45
  swing1Motion=Free
  swing2LimitDegrees=45
  swing2Motion=Free
  twistLimitDegrees=45
  twistMotion=Free
 }
 {
  bone1Name=dyn_r_backStrap_05
  bone2Name=dyn_r_frontStrap_07
  swing1LimitDegrees=45
  swing1Motion=Free
  swing2LimitDegrees=45
  swing2Motion=Free
  twistLimitDegrees=45
  twistMotion=Free
 }
 {
  bone1Name=dyn_l_frontStrap_MSTR
  bone2Name=pelvis
  swing1LimitDegrees=45
  swing1Motion=Free
  swing2LimitDegrees=45
  swing2Motion=Free
  twistLimitDegrees=45
  twistMotion=Free
 }
 {
  bone1Name=dyn_l_backStrap_MSTR
  bone2Name=pelvis
  swing1LimitDegrees=45
  swing1Motion=Free
  swing2LimitDegrees=45
  swing2Motion=Free
  twistLimitDegrees=45
  twistMotion=Free
 }
 {
  bone1Name=dyn_r_frontStrap_MSTR
  bone2Name=pelvis
  swing1LimitDegrees=45
  swing1Motion=Free
  swing2LimitDegrees=45
  swing2Motion=Free
  twistLimitDegrees=45
  twistMotion=Free
 }
 {
  bone1Name=dyn_r_backStrap_MSTR
  bone2Name=pelvis
  swing1LimitDegrees=45
  swing1Motion=Free
  swing2LimitDegrees=45
  swing2Motion=Free
  twistLimitDegrees=45
  twistMotion=Free
 }
 {
  bone1Name=thigh_twist_01_l
  bone2Name=thigh_l
  swing1LimitDegrees=45
  swing1Motion=Free
  swing2LimitDegrees=45
  swing2Motion=Free
  twistLimitDegrees=45
  twistMotion=Free
 }
 {
  bone1Name=thigh_twist_01_r
  bone2Name=thigh_r
  swing1LimitDegrees=45
  swing1Motion=Free
  swing2LimitDegrees=45
  swing2Motion=Free
  twistLimitDegrees=45
  twistMotion=Free
 }
 {
  bone1Name=calf_r
  bone2Name=thigh_r
  swing1LimitDegrees=45
  swing1Motion=Free
  swing2LimitDegrees=45
  swing2Motion=Free
  twistLimitDegrees=45
  twistMotion=Free
 }
]
