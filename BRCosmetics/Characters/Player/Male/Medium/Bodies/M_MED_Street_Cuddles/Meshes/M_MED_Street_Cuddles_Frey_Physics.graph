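# Physics asset for the Frey mesh of the male medium Street Cuddles cosmetic body, covering the upper neck, head and collar bones.
# Single-capsule bodies sit on two neck bones, the head and three dyn_ collar bones.
# The neck-to-head chain has free swing and twist.
# Each collar body is constrained directly to neck_02 with tighter angular limit values, apparently for secondary collar motion.
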PhysicsAsset /BRCosmetics/Characters/Player/Male/Medium/Bodies/M_MED_Street_Cuddles/Meshes/M_MED_Street_Cuddles_Frey_Physics
bodies {
 dyn_collar [{shapeType=Capsule, length=3.1123, radius=2.5792, center=<-3.9642 -0.0001 0>, rotation=<rot 90 0 0>}]
 dyn_l_collar [{shapeType=Capsule, length=3.1123, radius=2.5792, center=<3.7812 0.0001 0>, rotation=<rot 90 0 0>}]
 dyn_r_collar [{shapeType=Capsule, length=3.1123, radius=2.5792, center=<-3.8383 0 0>, rotation=<rot 90 0 0>}]
 head [{shapeType=Capsule, length=2, radius=2}]
 neck_01 [{shapeType=Capsule, length=2, radius=2}]
 neck_02 [{shapeType=Capsule, length=2, radius=2}]
}
constraints [
 {
  bone1Name=neck_02
  bone2Name=neck_01
  swing1LimitDegrees=45
  swing1Motion=Free
  swing2LimitDegrees=45
  swing2Motion=Free
  twistLimitDegrees=45
  twistMotion=Free
 }
 {
  bone1Name=head
  bone2Name=neck_02
  swing1LimitDegrees=45
  swing1Motion=Free
  swing2LimitDegrees=45
  swing2Motion=Free
  twistLimitDegrees=45
  twistMotion=Free
 }
 {bone1Name=dyn_collar, bone2Name=neck_02, swing1LimitDegrees=30, swing2LimitDegrees=30, twistLimitDegrees=30}
 {bone1Name=dyn_r_collar, bone2Name=neck_02, swing1LimitDegrees=30, swing2LimitDegrees=30, twistLimitDegrees=30}
 {bone1Name=dyn_l_collar, bone2Name=neck_02, swing1LimitDegrees=30, swing2LimitDegrees=30, twistLimitDegrees=30}
]
