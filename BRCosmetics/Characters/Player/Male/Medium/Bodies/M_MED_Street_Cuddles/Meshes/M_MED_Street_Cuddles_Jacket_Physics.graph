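# Physics asset for the Street Cuddles medium male jacket, setting up physics for seven dangling panel chains around the back, sides and front.
# Each chain runs from a tiny master capsule, attached to spine_04 with free rotation, through two box segments.
# Panel joints allow wide bend on one swing axis and little on the other or in twist, giving hinge-like flaps.
# Capsules on pelvis, spine and thighs approximate the body, joined by free-rotation constraints.
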PhysicsAsset /BRCosmetics/Characters/Player/Male/Medium/Bodies/M_MED_Street_Cuddles/Meshes/M_MED_Street_Cuddles_Jacket_Physics
bodies {
 dyn_jacket_bk_01 [{shapeType=Box, extentX=7.7783, extentY=12.6228, extentZ=3.6218, center=<4.0838 0 0.0001>}]
 dyn_jacket_bk_02 [{shapeType=Box, extentX=7.696, extentY=13.3871, extentZ=3.6218, center=<3.5805 0 -0.35>}]
 dyn_jacket_bk_MSTR [{shapeType=Capsule, length=0.505, radius=0.505, rotation=<rot 0 -90 0>}]
 dyn_jacket_bk_l_01 [{shapeType=Box, extentX=7.5373, extentY=11.9871, extentZ=4.7264, center=<3.7945 0 0.0001>}]
 dyn_jacket_bk_l_02 [{shapeType=Box, extentX=8.9604, extentY=12.7514, extentZ=5.0085, center=<4.6307 0 0>}]
 dyn_jacket_bk_l_MSTR [{shapeType=Capsule, length=0.505, radius=0.505, rotation=<rot 0 -90 0>}]
 dyn_jacket_bk_r_01 [{shapeType=Box, extentX=7.7783, extentY=17.9212, extentZ=3.6961, center=<4.0838 0 0.0001>, rotation=<rot 0 0 0.3415>}]
 dyn_jacket_bk_r_02 [{shapeType=Box, extentX=7.696, extentY=13.4495, extentZ=3.8419, center=<3.5805 0 0>, rotation=<rot 0 0 -1.9533>}]
 dyn_jacket_bk_r_MSTR [{shapeType=Capsule, length=0.505, radius=0.505, rotation=<rot 0 -90 0>}]
 dyn_jacket_fr_l_01 [{shapeType=Box, extentX=8.3745, extentY=12.6853, extentZ=2.4584, center=<4.2216 0 0.0001>}]
 dyn_jacket_fr_l_02 [{shapeType=Box, extentX=8.9604, extentY=13.4495, extentZ=2.4584, center=<4.6307 0 0>}]
 dyn_jacket_fr_l_MSTR [{shapeType=Capsule, length=0.505, radius=0.505, rotation=<rot 0 -90 0>}]
 dyn_jacket_fr_r_01 [{shapeType=Box, extentX=7.7783, extentY=12.6853, extentZ=2.4584, center=<4.0403 0 -0.8301>, rotation=<rot -9.9528 0 0>}]
 dyn_jacket_fr_r_02 [{shapeType=Box, extentX=7.696, extentY=13.4495, extentZ=2.4584, center=<3.5805 0 0>}]
 dyn_jacket_fr_r_MSTR [{shapeType=Capsule, length=0.505, radius=0.505, rotation=<rot 0 -90 0>}]
 dyn_jacket_l_01 [{shapeType=Box, extentX=7.5373, extentY=12.6853, extentZ=4.536, center=<3.7945 0 0.0001>}]
 dyn_jacket_l_02 [{shapeType=Box, extentX=8.9604, extentY=13.4495, extentZ=4.536, center=<4.6307 0 0>}]
 dyn_jacket_l_MSTR [{shapeType=Capsule, length=0.505, radius=0.505, rotation=<rot 0 -90 0>}]
 dyn_jacket_r_01 [{shapeType=Box, extentX=7.7783, extentY=12.6853, extentZ=2.4584, center=<4.0403 0 -0.8301>, rotation=<rot -9.9528 0 0>}]
 dyn_jacket_r_02 [{shapeType=Box, extentX=7.696, extentY=13.4495, extentZ=2.4584, center=<3.5805 0 0>}]
 dyn_jacket_r_MSTR [{shapeType=Capsule, length=0.505, radius=0.505, rotation=<rot 0 -90 0>}]
 pelvis [
  {shapeType=Capsule, length=5.5532, radius=14.0128, center=<-0.9194 -1.7087 0.0001>}
  {shapeType=Capsule, length=11.1355, radius=8.4304, center=<-0.7538 -1.5946 15.3953>, rotation=<rot 0 0 -85.0166>}
 ]
 spine_01 [{shapeType=Capsule, length=6.6958, radius=12, center=<0 -0.5406 0.0001>}]
 spine_02 [{shapeType=Capsule, length=6.6958, radius=13, center=<0 -0.5406 0.0001>}]
 spine_04 [{shapeType=Capsule, length=7.4312, radius=12}]
 thigh_l [{shapeType=Capsule, length=28.1435, radius=9.1246, center=<-17.0688 0.0002 0>, rotation=<rot 90 0 0>}]
 thigh_r [{shapeType=Capsule, length=28.1435, radius=9.1246, center=<16.403 -0.0002 0>, rotation=<rot 90 0 0>}]
}
constraints [
 {
  bone1Name=spine_01
  bone2Name=pelvis
  swing1LimitDegrees=45
  swing1Motion=Free
  swing2LimitDegrees=45
  swing2Motion=Free
  twistLimitDegrees=45
  twistMotion=Free
 }
 {
  bone1Name=spine_02
  bone2Name=spine_01
  swing1LimitDegrees=45
  swing1Motion=Free
  swing2LimitDegrees=45
  swing2Motion=Free
  twistLimitDegrees=45
  twistMotion=Free
 }
 {
  bone1Name=spine_04
  bone2Name=spine_02
  swing1LimitDegrees=45
  swing1Motion=Free
  swing2LimitDegrees=45
  swing2Motion=Free
  twistLimitDegrees=45
  twistMotion=Free
 }
 {bone1Name=dyn_jacket_bk_01, bone2Name=dyn_jacket_bk_MSTR, swing1LimitDegrees=5, swing2LimitDegrees=45, twistLimitDegrees=5}
 {bone1Name=dyn_jacket_bk_02, bone2Name=dyn_jacket_bk_01, swing1LimitDegrees=5, swing2LimitDegrees=45, twistLimitDegrees=5}
 {bone1Name=dyn_jacket_l_01, bone2Name=dyn_jacket_l_MSTR, swing1LimitDegrees=5, swing2LimitDegrees=45, twistLimitDegrees=5}
 {bone1Name=dyn_jacket_l_02, bone2Name=dyn_jacket_l_01, swing1LimitDegrees=5, swing2LimitDegrees=45, twistLimitDegrees=5}
 {bone1Name=dyn_jacket_bk_r_01, bone2Name=dyn_jacket_bk_r_MSTR, swing1LimitDegrees=5, swing2LimitDegrees=45, twistLimitDegrees=5}
 {bone1Name=dyn_jacket_bk_r_02, bone2Name=dyn_jacket_bk_r_01, swing1LimitDegrees=5, swing2LimitDegrees=45, twistLimitDegrees=5}
 {bone1Name=dyn_jacket_fr_r_01, bone2Name=dyn_jacket_fr_r_MSTR, swing1LimitDegrees=5, swing2LimitDegrees=45, twistLimitDegrees=5}
 {bone1Name=dyn_jacket_fr_r_02, bone2Name=dyn_jacket_fr_r_01, swing1LimitDegrees=5, swing2LimitDegrees=45, twistLimitDegrees=5}
 {bone1Name=dyn_jacket_fr_l_01, bone2Name=dyn_jacket_fr_l_MSTR, swing1LimitDegrees=5, swing2LimitDegrees=45, twistLimitDegrees=5}
 {bone1Name=dyn_jacket_fr_l_02, bone2Name=dyn_jacket_fr_l_01, swing1LimitDegrees=5, swing2LimitDegrees=45, twistLimitDegrees=5}
 {bone1Name=dyn_jacket_bk_l_01, bone2Name=dyn_jacket_bk_l_MSTR, swing1LimitDegrees=5, swing2LimitDegrees=45, twistLimitDegrees=5}
 {bone1Name=dyn_jacket_bk_l_02, bone2Name=dyn_jacket_bk_l_01, swing1LimitDegrees=5, swing2LimitDegrees=45, twistLimitDegrees=5}
 {bone1Name=dyn_jacket_r_01, bone2Name=dyn_jacket_r_MSTR, swing1LimitDegrees=5, swing2LimitDegrees=45, twistLimitDegrees=5}
 {bone1Name=dyn_jacket_r_02, bone2Name=dyn_jacket_r_01, swing1LimitDegrees=5, swing2LimitDegrees=45, twistLimitDegrees=5}
 {
  bone1Name=dyn_jacket_l_MSTR
  bone2Name=spine_04
  swing1LimitDegrees=45
  swing1Motion=Free
  swing2LimitDegrees=45
  swing2Motion=Free
  twistLimitDegrees=45
  twistMotion=Free
 }
 {
  bone1Name=dyn_jacket_bk_l_MSTR
  bone2Name=spine_04
  swing1LimitDegrees=45
  swing1Motion=Free
  swing2LimitDegrees=45
  swing2Motion=Free
  twistLimitDegrees=45
  twistMotion=Free
 }
 {
  bone1Name=dyn_jacket_bk_MSTR
  bone2Name=spine_04
  swing1LimitDegrees=45
  swing1Motion=Free
  swing2LimitDegrees=45
  swing2Motion=Free
  twistLimitDegrees=45
  twistMotion=Free
 }
 {
  bone1Name=dyn_jacket_bk_r_MSTR
  bone2Name=spine_04
  swing1LimitDegrees=45
  swing1Motion=Free
  swing2LimitDegrees=45
  swing2Motion=Free
  twistLimitDegrees=45
  twistMotion=Free
 }
 {
  bone1Name=dyn_jacket_r_MSTR
  bone2Name=spine_04
  swing1LimitDegrees=45
  swing1Motion=Free
  swing2LimitDegrees=45
  swing2Motion=Free
  twistLimitDegrees=45
  twistMotion=Free
 }
 {
  bone1Name=dyn_jacket_fr_l_MSTR
  bone2Name=spine_04
  swing1LimitDegrees=45
  swing1Motion=Free
  swing2LimitDegrees=45
  swing2Motion=Free
  twistLimitDegrees=45
  twistMotion=Free
 }
 {
  bone1Name=dyn_jacket_fr_r_MSTR
  bone2Name=spine_04
  swing1LimitDegrees=45
  swing1Motion=Free
  swing2LimitDegrees=45
  swing2Motion=Free
  twistLimitDegrees=45
  twistMotion=Free
 }
 {
  bone1Name=thigh_l
  bone2Name=pelvis
  swing1LimitDegrees=45
  swing1Motion=Free
  swing2LimitDegrees=45
  swing2Motion=Free
  twistLimitDegrees=45
  twistMotion=Free
 }
 {
  bone1Name=thigh_r
  bone2Name=pelvis
  swing1LimitDegrees=45
  swing1Motion=Free
  swing2LimitDegrees=45
  swing2Motion=Free
  twistLimitDegrees=45
  twistMotion=Free
 }
]
