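# Physics asset for the side bag trinket on the male medium Street Cuddles body, defining bodies for the bag, its straps, pelvis and thighs.
# The bag is a single angled box constrained to a small master capsule, with narrow swing-1 and twist limits and a wider swing-2 limit.
# That master, four small strap anchors and both thighs attach to the pelvis with free angular motion, and thigh twist capsules extend the leg volumes.
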PhysicsAsset /BRCosmetics/Characters/Player/Male/Medium/Bodies/M_MED_Street_Cuddles/Meshes/M_MED_Street_Cuddles_SideBagTrinket_Physics
bodies {
 dyn_bag_01 [{shapeType=Box, extentX=11.2441, extentY=21.7709, extentZ=2.5469, center=<21.5 -0.3247 -0.6698>, rotation=<rot -48 2.9218 -4.2719>}]
 dyn_bag_MSTR [{shapeType=Capsule, length=2, radius=2, rotation=<rot 90 0 0>}]
 dyn_l_backStrap_MSTR [{shapeType=Capsule, length=1, radius=1}]
 dyn_l_frontStrap_MSTR [{shapeType=Capsule, length=1, radius=1}]
 dyn_r_backStrap_MSTR [{shapeType=Capsule, length=1, radius=1}]
 dyn_r_frontStrap_MSTR [{shapeType=Capsule, length=1, radius=1}]
 pelvis [{shapeType=Capsule, length=10.0142, radius=9.4323}]
 thigh_l [{shapeType=Capsule, length=24.6239, radius=8.7119, center=<-19.8239 0.0003 0.3471>, rotation=<rot 90 0 0>}]
 thigh_r [{shapeType=Capsule, length=43.9919, radius=7.2862, center=<20.1068 0.0006 0.3471>, rotation=<rot 90 0 0>}]
 thigh_twist_01_l [{shapeType=Capsule, length=24.4346, radius=8.7383, center=<1.5154 0 0.3471>, rotation=<rot 90 0 0>}]
 thigh_twist_01_r [{shapeType=Capsule, length=29.2864, radius=3.9502, center=<-1.7034 0.0009 0.3471>, rotation=<rot 90 0 0>}]
}
constraints [
 {
  bone1Name=thigh_twist_01_r
  bone2Name=thigh_r
  swing1LimitDegrees=45
  swing1Motion=Free
  swing2LimitDegrees=45
  swing2Motion=Free
  twistLimitDegrees=45
  twistMotion=Free
 }
 {
  bone1Name=thigh_twist_01_l
  bone2Name=thigh_l
  swing1LimitDegrees=45
  swing1Motion=Free
  swing2LimitDegrees=45
  swing2Motion=Free
  twistLimitDegrees=45
  twistMotion=Free
 }
 {bone1Name=dyn_bag_01, bone2Name=dyn_bag_MSTR, swing1LimitDegrees=5, swing2LimitDegrees=60, twistLimitDegrees=5}
 {
  bone1Name=thigh_l
  bone2Name=pelvis
  swing1LimitDegrees=45
  swing1Motion=Free
  swing2LimitDegrees=45
  swing2Motion=Free
  twistLimitDegrees=45
  twistMotion=Free
 }
 {
  bone1Name=dyn_l_frontStrap_MSTR
  bone2Name=pelvis
  swing1LimitDegrees=45
  swing1Motion=Free
  swing2LimitDegrees=45
  swing2Motion=Free
  twistLimitDegrees=45
  twistMotion=Free
 }
 {
  bone1Name=thigh_r
  bone2Name=pelvis
  swing1LimitDegrees=45
  swing1Motion=Free
  swing2LimitDegrees=45
  swing2Motion=Free
  twistLimitDegrees=45
  twistMotion=Free
 }
 {
  bone1Name=dyn_r_frontStrap_MSTR
  bone2Name=pelvis
  swing1LimitDegrees=45
  swing1Motion=Free
  swing2LimitDegrees=45
  swing2Motion=Free
  twistLimitDegrees=45
  twistMotion=Free
 }
 {
  bone1Name=dyn_l_backStrap_MSTR
  bone2Name=pelvis
  swing1LimitDegrees=45
  swing1Motion=Free
  swing2LimitDegrees=45
  swing2Motion=Free
  twistLimitDegrees=45
  twistMotion=Free
 }
 {
  bone1Name=dyn_bag_MSTR
  bone2Name=pelvis
  swing1LimitDegrees=45
  swing1Motion=Free
  swing2LimitDegrees=45
  swing2Motion=Free
  twistLimitDegrees=45
  twistMotion=Free
 }
 {
  bone1Name=dyn_r_backStrap_MSTR
  bone2Name=pelvis
  swing1LimitDegrees=45
  swing1Motion=Free
  swing2LimitDegrees=45
  swing2Motion=Free
  twistLimitDegrees=45
  twistMotion=Free
 }
]
