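# Physics asset for the suspenders on the Street Cuddles medium male player body, with capsule and box bodies for four strap chains.
# Left and right, front and back chains each run from a MSTR body through limited swing and twist joints.
# Free-angular constraints cross-link front and back chain ends on each side and tie the MSTR bodies to the pelvis.
# Pelvis, thigh, thigh-twist and bag capsules complete the setup, also joined by free-angular constraints.
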PhysicsAsset /BRCosmetics/Characters/Player/Male/Medium/Bodies/M_MED_Street_Cuddles/Meshes/M_MED_Street_Cuddles_Suspenders_Physics
bodies {
 dyn_bag_03 [{shapeType=Capsule, length=19.9251, radius=4.6269, center=<-22.9932 -0.013 -20.1276>, rotation=<rot 0 3.1247 89.9999>}]
 dyn_l_backStrap_01 [{shapeType=Capsule, length=5.8478, radius=2.6051, shapeName=dyn_l_backStrap_01_capsule, center=<3.012 0 0>, rotation=<rot 0 0 90>}]
 dyn_l_backStrap_02 [{shapeType=Capsule, length=6.0639, radius=2.3891, shapeName=dyn_l_backStrap_02_capsule, center=<3.012 0 0>, rotation=<rot 0 0 90>}]
 dyn_l_backStrap_03 [{shapeType=Capsule, length=6.3855, radius=2.0675, shapeName=dyn_l_backStrap_03_capsule, center=<3.012 0 0>, rotation=<rot 0 0 90>}]
 dyn_l_backStrap_04 [{shapeType=Capsule, length=6.3855, radius=2.0675, shapeName=dyn_l_backStrap_04_capsule, center=<3.012 0 0>, rotation=<rot 0 0 90>}]
 dyn_l_backStrap_05 [{shapeType=Box, extentX=4.5064, extentY=7.2297, extentZ=3.2519, center=<2.3165 0 0>}]
 dyn_l_backStrap_06 [{shapeType=Box, extentX=4.2993, extentY=6.81, extentZ=3.2519, center=<-1.4263 1.2079 -1.2818>, rotation=<rot 30.6462 -37.5085 -5.3725>}]
 dyn_l_backStrap_MSTR [{shapeType=Capsule, length=1, radius=1}]
 dyn_l_frontStrap_01 [{shapeType=Capsule, length=5.7032, radius=1.7287, shapeName=dyn_l_frontStrap_01_capsule, center=<2.9504 0 0>, rotation=<rot 0 0 90>}]
 dyn_l_frontStrap_02 [{shapeType=Capsule, length=5.7032, radius=1.7287, shapeName=dyn_l_frontStrap_02_capsule, center=<2.9504 0 0>, rotation=<rot 0 0 90>}]
 dyn_l_frontStrap_03 [{shapeType=Capsule, length=5.5464, radius=1.8855, shapeName=dyn_l_frontStrap_03_capsule, center=<2.9504 0 0>, rotation=<rot 0 0 90>}]
 dyn_l_frontStrap_04 [{shapeType=Capsule, length=5.5464, radius=1.8855, shapeName=dyn_l_frontStrap_04_capsule, center=<2.9504 0 0>, rotation=<rot 0 0 90>}]
 dyn_l_frontStrap_05 [{shapeType=Box, extentX=3.3135, extentY=6.0184, extentZ=3.2519, center=<1.7555 0 0.0001>}]
 dyn_l_frontStrap_06 [{shapeType=Box, extentX=4.4668, extentY=7.2297, extentZ=3.2519, center=<2.2784 0 -0.0001>}]
 dyn_l_frontStrap_07 [{shapeType=Box, extentX=7.0279, extentY=7.2297, extentZ=3.2519, center=<3.2607 -0.9485 -1.1025>, rotation=<rot -13.7988 -3.5008 9.6983>}]
 dyn_l_frontStrap_MSTR [{shapeType=Capsule, length=1, radius=1}]
 dyn_r_backStrap_01 [{shapeType=Capsule, length=3.5341, radius=3.8058, shapeName=dyn_r_backStrap_01_capsule, center=<3.6216 0 -0.0001>, rotation=<rot 0 0 90>}]
 dyn_r_backStrap_02 [{shapeType=Capsule, length=4.3626, radius=2.9772, shapeName=dyn_r_backStrap_02_capsule, center=<3.6216 0 -0.0001>, rotation=<rot 0 0 90>}]
 dyn_r_backStrap_03 [{shapeType=Capsule, length=5.7996, radius=1.5402, shapeName=dyn_r_backStrap_03_capsule, center=<3.6216 0 -0.0001>, rotation=<rot 0 0 90>}]
 dyn_r_backStrap_04 [{shapeType=Capsule, length=5.7996, radius=1.5402, shapeName=dyn_r_backStrap_04_capsule, center=<3.6216 0 -0.0001>, rotation=<rot 0 0 90>}]
 dyn_r_backStrap_05 [{shapeType=Box, extentX=4.5054, extentY=7.2258, extentZ=5.7812, center=<2.4291 0 0>, rotation=<rot 0 0 -21.2961>}]
 dyn_r_backStrap_06 [{shapeType=Box, extentX=3.0528, extentY=6.7421, extentZ=5.7812, center=<1.593 0 0>, rotation=<rot -17.0894 4.9149 -14.9181>}]
 dyn_r_backStrap_MSTR [{shapeType=Capsule, length=1, radius=1}]
 dyn_r_frontStrap_01 [{
   shapeType=Capsule
   length=6.2852
   radius=1.5193
   shapeName=dyn_r_frontStrap_01_capsule
   center=<6.5845 -0.0115 -0.1105>
   rotation=<rot 0 0 90>
  }]
 dyn_r_frontStrap_02 [{
   shapeType=Capsule
   length=6.5677
   radius=1.5129
   shapeName=dyn_r_frontStrap_02_capsule
   center=<3.5675 -0.0661 -0.1463>
   rotation=<rot 0 0 90>
  }]
 dyn_r_frontStrap_03 [{
   shapeType=Capsule
   length=6.3513
   radius=1.7293
   shapeName=dyn_r_frontStrap_02_capsule
   center=<3.0577 -0.0661 -0.1464>
   rotation=<rot 0 0 90>
  }]
 dyn_r_frontStrap_04 [{
   shapeType=Capsule
   length=5.0326
   radius=1.7474
   shapeName=dyn_r_frontStrap_04_capsule
   center=<1.6663 -0.1776 0.1492>
   rotation=<rot 0 0 90>
  }]
 dyn_r_frontStrap_05 [{shapeType=Box, extentX=3.0409, extentY=6.5477, extentZ=5.7812, center=<1.6866 0 0>, rotation=<rot 0 0 19.6769>}]
 dyn_r_frontStrap_06 [{shapeType=Box, extentX=4.3145, extentY=7.2258, extentZ=3.7817, center=<2.0334 0 0.0001>, rotation=<rot 0 0 23.6857>}]
 dyn_r_frontStrap_07 [{shapeType=Box, extentX=4.9557, extentY=5.4409, extentZ=5.7812, center=<2.4013 0 0>, rotation=<rot 0 0 1.6431>}]
 dyn_r_frontStrap_MSTR [{shapeType=Capsule, length=1, radius=1}]
 pelvis [
  {shapeType=Capsule, length=11.0232, radius=8.5428, center=<-4.162 -1.6524 0.0002>}
  {shapeType=Capsule, length=19.3632, radius=5.4484, center=<-3.7321 0.5218 21.7047>, rotation=<rot 0 0 -92.3929>}
 ]
 thigh_l [{shapeType=Capsule, length=25.3427, radius=7.9931, center=<-19.8239 0.3227 0.3471>, rotation=<rot 90 0 0>}]
 thigh_r [{shapeType=Capsule, length=25.903, radius=7.3337, center=<20.1068 0.0006 0.3471>, rotation=<rot 90 0 0>}]
 thigh_twist_01_l [{shapeType=Capsule, length=24.2496, radius=8.9232, center=<1.5154 0.3223 1.4338>, rotation=<rot 90 0 0>}]
 thigh_twist_01_r [{shapeType=Capsule, length=25.903, radius=7.3337, center=<-1.7034 0.0009 0.3471>, rotation=<rot 90 0 0>}]
}
constraints [
 {
  bone1Name=thigh_l
  bone2Name=pelvis
  swing1LimitDegrees=45
  swing1Motion=Free
  swing2LimitDegrees=45
  swing2Motion=Free
  twistLimitDegrees=45
  twistMotion=Free
 }
 {
  bone1Name=thigh_r
  bone2Name=pelvis
  swing1LimitDegrees=45
  swing1Motion=Free
  swing2LimitDegrees=45
  swing2Motion=Free
  twistLimitDegrees=45
  twistMotion=Free
 }
 {bone1Name=dyn_l_backStrap_01, bone2Name=dyn_l_backStrap_MSTR, swing1LimitDegrees=20, swing2LimitDegrees=20, twistLimitDegrees=10}
 {bone1Name=dyn_l_frontStrap_01, bone2Name=dyn_l_frontStrap_MSTR, swing1LimitDegrees=20, swing2LimitDegrees=20, twistLimitDegrees=10}
 {bone1Name=dyn_l_frontStrap_02, bone2Name=dyn_l_frontStrap_01, swing1LimitDegrees=60, swing2LimitDegrees=60, twistLimitDegrees=10}
 {bone1Name=dyn_l_frontStrap_03, bone2Name=dyn_l_frontStrap_02, swing1LimitDegrees=60, swing2LimitDegrees=60, twistLimitDegrees=10}
 {bone1Name=dyn_l_frontStrap_04, bone2Name=dyn_l_frontStrap_03, swing1LimitDegrees=60, swing2LimitDegrees=60, twistLimitDegrees=10}
 {bone1Name=dyn_l_frontStrap_05, bone2Name=dyn_l_frontStrap_04, swing1LimitDegrees=60, swing2LimitDegrees=60, twistLimitDegrees=10}
 {bone1Name=dyn_l_frontStrap_06, bone2Name=dyn_l_frontStrap_05, swing1LimitDegrees=60, swing2LimitDegrees=60, twistLimitDegrees=10}
 {bone1Name=dyn_l_frontStrap_07, bone2Name=dyn_l_frontStrap_06, swing1LimitDegrees=60, swing2LimitDegrees=60, twistLimitDegrees=10}
 {bone1Name=dyn_l_backStrap_02, bone2Name=dyn_l_backStrap_01, swing1LimitDegrees=60, swing2LimitDegrees=60, twistLimitDegrees=10}
 {bone1Name=dyn_l_backStrap_03, bone2Name=dyn_l_backStrap_02, swing1LimitDegrees=60, swing2LimitDegrees=60, twistLimitDegrees=10}
 {bone1Name=dyn_l_backStrap_04, bone2Name=dyn_l_backStrap_03, swing1LimitDegrees=60, swing2LimitDegrees=60, twistLimitDegrees=10}
 {bone1Name=dyn_l_backStrap_05, bone2Name=dyn_l_backStrap_04, swing1LimitDegrees=60, swing2LimitDegrees=60, twistLimitDegrees=10}
 {bone1Name=dyn_l_backStrap_06, bone2Name=dyn_l_backStrap_05, swing1LimitDegrees=60, swing2LimitDegrees=60, twistLimitDegrees=10}
 {bone1Name=dyn_r_frontStrap_01, bone2Name=dyn_r_frontStrap_MSTR, swing1LimitDegrees=10, swing2LimitDegrees=20, twistLimitDegrees=10}
 {bone1Name=dyn_r_frontStrap_02, bone2Name=dyn_r_frontStrap_01, swing1LimitDegrees=20, swing2LimitDegrees=60, twistLimitDegrees=10}
 {bone1Name=dyn_r_frontStrap_03, bone2Name=dyn_r_frontStrap_02, swing1LimitDegrees=30, swing2LimitDegrees=60, twistLimitDegrees=10}
 {bone1Name=dyn_r_frontStrap_04, bone2Name=dyn_r_frontStrap_03, swing1LimitDegrees=60, swing2LimitDegrees=60, twistLimitDegrees=10}
 {bone1Name=dyn_r_frontStrap_05, bone2Name=dyn_r_frontStrap_04, swing1LimitDegrees=60, swing2LimitDegrees=60, twistLimitDegrees=10}
 {bone1Name=dyn_r_frontStrap_06, bone2Name=dyn_r_frontStrap_05, swing1LimitDegrees=60, swing2LimitDegrees=60, twistLimitDegrees=10}
 {bone1Name=dyn_r_frontStrap_07, bone2Name=dyn_r_frontStrap_06, swing1LimitDegrees=60, swing2LimitDegrees=60, twistLimitDegrees=10}
 {bone1Name=dyn_r_backStrap_01, bone2Name=dyn_r_backStrap_MSTR, swing1LimitDegrees=20, swing2LimitDegrees=20, twistLimitDegrees=10}
 {bone1Name=dyn_r_backStrap_02, bone2Name=dyn_r_backStrap_01, swing1LimitDegrees=60, swing2LimitDegrees=60, twistLimitDegrees=10}
 {bone1Name=dyn_r_backStrap_03, bone2Name=dyn_r_backStrap_02, swing1LimitDegrees=60, swing2LimitDegrees=60, twistLimitDegrees=10}
 {bone1Name=dyn_r_backStrap_04, bone2Name=dyn_r_backStrap_03, swing1LimitDegrees=60, swing2LimitDegrees=60, twistLimitDegrees=10}
 {bone1Name=dyn_r_backStrap_05, bone2Name=dyn_r_backStrap_04, swing1LimitDegrees=60, swing2LimitDegrees=60, twistLimitDegrees=10}
 {bone1Name=dyn_r_backStrap_06, bone2Name=dyn_r_backStrap_05, swing1LimitDegrees=60, swing2LimitDegrees=60, twistLimitDegrees=10}
 {
  bone1Name=dyn_l_frontStrap_06
  bone2Name=dyn_l_backStrap_06
  swing1LimitDegrees=45
  swing1Motion=Free
  swing2LimitDegrees=45
  swing2Motion=Free
  twistLimitDegrees=45
  twistMotion=Free
 }
 {
  bone1Name=dyn_l_backStrap_05
  bone2Name=dyn_l_frontStrap_07
  swing1LimitDegrees=45
  swing1Motion=Free
  swing2LimitDegrees=45
  swing2Motion=Free
  twistLimitDegrees=45
  twistMotion=Free
 }
 {
  bone1Name=dyn_r_backStrap_06
  bone2Name=dyn_r_frontStrap_06
  swing1LimitDegrees=45
  swing1Motion=Free
  swing2LimitDegrees=45
  swing2Motion=Free
  twistLimitDegrees=45
  twistMotion=Free
 }
 {
  bone1Name=dyn_r_backStrap_05
  bone2Name=dyn_r_frontStrap_07
  swing1LimitDegrees=45
  swing1Motion=Free
  swing2LimitDegrees=45
  swing2Motion=Free
  twistLimitDegrees=45
  twistMotion=Free
 }
 {
  bone1Name=dyn_l_frontStrap_MSTR
  bone2Name=pelvis
  swing1LimitDegrees=45
  swing1Motion=Free
  swing2LimitDegrees=45
  swing2Motion=Free
  twistLimitDegrees=45
  twistMotion=Free
 }
 {
  bone1Name=dyn_l_backStrap_MSTR
  bone2Name=pelvis
  swing1LimitDegrees=45
  swing1Motion=Free
  swing2LimitDegrees=45
  swing2Motion=Free
  twistLimitDegrees=45
  twistMotion=Free
 }
 {
  bone1Name=dyn_r_frontStrap_MSTR
  bone2Name=pelvis
  swing1LimitDegrees=45
  swing1Motion=Free
  swing2LimitDegrees=45
  swing2Motion=Free
  twistLimitDegrees=45
  twistMotion=Free
 }
 {
  bone1Name=dyn_r_backStrap_MSTR
  bone2Name=pelvis
  swing1LimitDegrees=45
  swing1Motion=Free
  swing2LimitDegrees=45
  swing2Motion=Free
  twistLimitDegrees=45
  twistMotion=Free
 }
 {
  bone1Name=thigh_twist_01_r
  bone2Name=thigh_r
  swing1LimitDegrees=45
  swing1Motion=Free
  swing2LimitDegrees=45
  swing2Motion=Free
  twistLimitDegrees=45
  twistMotion=Free
 }
 {
  bone1Name=thigh_twist_01_l
  bone2Name=thigh_l
  swing1LimitDegrees=45
  swing1Motion=Free
  swing2LimitDegrees=45
  swing2Motion=Free
  twistLimitDegrees=45
  twistMotion=Free
 }
 {
  bone1Name=dyn_bag_03
  bone2Name=pelvis
  swing1LimitDegrees=45
  swing1Motion=Free
  swing2LimitDegrees=45
  swing2Motion=Free
  twistLimitDegrees=45
  twistMotion=Free
 }
]
