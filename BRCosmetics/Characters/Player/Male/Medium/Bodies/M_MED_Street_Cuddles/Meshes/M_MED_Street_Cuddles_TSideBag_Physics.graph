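# Physics asset for the side bag and hanging charm on the M_MED_Street_Cuddles male cosmetic body.
# Capsule bodies cover the bag, charm, four strap roots, pelvis and right thigh, with two overlapping shapes on the thigh.
# The bag chain runs from pelvis through master and three segments into a two-link charm.
# Only the charm joints keep default, limit-based angular motion; all other constraints leave swing and twist free.
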PhysicsAsset /BRCosmetics/Characters/Player/Male/Medium/Bodies/M_MED_Street_Cuddles/Meshes/M_MED_Street_Cuddles_TSideBag_Physics
bodies {
 dyn_bag_01 [{shapeType=Capsule, length=2, radius=2}]
 dyn_bag_02 [{shapeType=Capsule, length=2, radius=2}]
 dyn_bag_03 [{shapeType=Capsule, length=2, radius=2}]
 dyn_bag_MSTR [{shapeType=Capsule, length=2, radius=2, rotation=<rot 90 0 0>}]
 dyn_charm_01 [{shapeType=Capsule, length=2.4008, radius=0.7867, center=<1.8423 0 0>, rotation=<rot 90 0 0>}]
 dyn_charm_02 [{shapeType=Capsule, length=2.5539, radius=3.0694, center=<4.2895 0 0>, rotation=<rot 90 0 0>}]
 dyn_l_backStrap_MSTR [{shapeType=Capsule, length=1, radius=1}]
 dyn_l_frontStrap_MSTR [{shapeType=Capsule, length=1, radius=1}]
 dyn_r_backStrap_MSTR [{shapeType=Capsule, length=1, radius=1}]
 dyn_r_frontStrap_MSTR [{shapeType=Capsule, length=1, radius=1}]
 pelvis [{shapeType=Capsule, length=10.0142, radius=9.4323}]
 thigh_r [
  {shapeType=Capsule, length=43.9919, radius=7.2862, center=<20.1068 0.0006 0.3471>, rotation=<rot 90 0 0>}
  {shapeType=Capsule, length=5.5218, radius=7.2862, center=<4.7426 -0.2189 7.7969>, rotation=<rot 90 0 0>}
 ]
 thigh_twist_01_r [{shapeType=Capsule, length=29.2864, radius=3.9502, center=<-1.7034 0.0009 0.3471>, rotation=<rot 90 0 0>}]
}
constraints [
 {
  bone1Name=thigh_twist_01_r
  bone2Name=thigh_r
  swing1LimitDegrees=45
  swing1Motion=Free
  swing2LimitDegrees=45
  swing2Motion=Free
  twistLimitDegrees=45
  twistMotion=Free
 }
 {
  bone1Name=dyn_l_frontStrap_MSTR
  bone2Name=pelvis
  swing1LimitDegrees=45
  swing1Motion=Free
  swing2LimitDegrees=45
  swing2Motion=Free
  twistLimitDegrees=45
  twistMotion=Free
 }
 {
  bone1Name=thigh_r
  bone2Name=pelvis
  swing1LimitDegrees=45
  swing1Motion=Free
  swing2LimitDegrees=45
  swing2Motion=Free
  twistLimitDegrees=45
  twistMotion=Free
 }
 {
  bone1Name=dyn_r_frontStrap_MSTR
  bone2Name=pelvis
  swing1LimitDegrees=45
  swing1Motion=Free
  swing2LimitDegrees=45
  swing2Motion=Free
  twistLimitDegrees=45
  twistMotion=Free
 }
 {
  bone1Name=dyn_l_backStrap_MSTR
  bone2Name=pelvis
  swing1LimitDegrees=45
  swing1Motion=Free
  swing2LimitDegrees=45
  swing2Motion=Free
  twistLimitDegrees=45
  twistMotion=Free
 }
 {
  bone1Name=dyn_bag_MSTR
  bone2Name=pelvis
  swing1LimitDegrees=45
  swing1Motion=Free
  swing2LimitDegrees=45
  swing2Motion=Free
  twistLimitDegrees=45
  twistMotion=Free
 }
 {
  bone1Name=dyn_r_backStrap_MSTR
  bone2Name=pelvis
  swing1LimitDegrees=45
  swing1Motion=Free
  swing2LimitDegrees=45
  swing2Motion=Free
  twistLimitDegrees=45
  twistMotion=Free
 }
 {bone1Name=dyn_charm_01, bone2Name=dyn_bag_03, swing1LimitDegrees=45, swing2LimitDegrees=5, twistLimitDegrees=20}
 {bone1Name=dyn_charm_02, bone2Name=dyn_charm_01, swing1LimitDegrees=45, swing2LimitDegrees=5, twistLimitDegrees=20}
 {
  bone1Name=dyn_bag_03
  bone2Name=dyn_bag_02
  swing1LimitDegrees=45
  swing1Motion=Free
  swing2LimitDegrees=45
  swing2Motion=Free
  twistLimitDegrees=45
  twistMotion=Free
 }
 {
  bone1Name=dyn_bag_02
  bone2Name=dyn_bag_01
  swing1LimitDegrees=45
  swing1Motion=Free
  swing2LimitDegrees=45
  swing2Motion=Free
  twistLimitDegrees=45
  twistMotion=Free
 }
 {
  bone1Name=dyn_bag_01
  bone2Name=dyn_bag_MSTR
  swing1LimitDegrees=45
  swing1Motion=Free
  swing2LimitDegrees=45
  swing2Motion=Free
  twistLimitDegrees=45
  twistMotion=Free
 }
]
