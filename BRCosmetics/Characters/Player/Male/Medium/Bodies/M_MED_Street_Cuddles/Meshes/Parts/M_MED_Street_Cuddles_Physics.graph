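# Physics asset for the male medium Street Cuddles body cosmetic, providing secondary-motion bodies for its hood, two ears and two drawstrings.
# Capsules on spine_04 and spine_05 form the torso, with two extra angled side capsules on spine_05.
# Tiny base and MSTR capsules link to spine_05 with free angular motion and anchor the dynamic parts.
# Hood pieces and ears are single limited joints, while each string is a three-link chain; the hood back has only a base.
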PhysicsAsset /BRCosmetics/Characters/Player/Male/Medium/Bodies/M_MED_Street_Cuddles/Meshes/Parts/M_MED_Street_Cuddles_Physics
bodies {
 dyn_hood_back_base [{shapeType=Capsule, length=0.505, radius=0.505, rotation=<rot 0 -90 0>}]
 dyn_hood_front [{shapeType=Capsule, length=3, radius=2, center=<3.3211 0 0>, rotation=<rot 90 0 0>}]
 dyn_hood_front_base [{shapeType=Capsule, length=0.505, radius=0.505, rotation=<rot 0 -90 0>}]
 dyn_hood_l [{shapeType=Capsule, length=2.9517, radius=2.0483, center=<3.3211 0 0>, rotation=<rot 90 0 0>}]
 dyn_hood_l_base [{shapeType=Capsule, length=0.505, radius=0.505, rotation=<rot 0 -90 0>}]
 dyn_hood_r [{shapeType=Capsule, length=3, radius=2, center=<3.3211 0 0>, rotation=<rot 90 0 0>}]
 dyn_hood_r_base [{shapeType=Capsule, length=0.505, radius=0.505, rotation=<rot 0 -90 0>}]
 dyn_hood_top [{shapeType=Capsule, length=3, radius=2, center=<3.3211 0 0>, rotation=<rot 90 0 0>}]
 dyn_hood_top_base [{shapeType=Capsule, length=0.505, radius=0.505, rotation=<rot 0 -90 0>}]
 dyn_l_ear [{shapeType=Capsule, length=3.1839, radius=3.1828, center=<0.9917 0 0.2729>}]
 dyn_l_ear_MSTR [{shapeType=Capsule, length=0.505, radius=0.505}]
 dyn_l_string_01 [{shapeType=Capsule, length=3.4493, radius=0.5507, center=<1.6111 -0.0861 -0.0002>, rotation=<rot 90 0 0>}]
 dyn_l_string_02 [{shapeType=Capsule, length=3.4657, radius=0.8628, center=<1.8319 -0.0859 -0.0002>, rotation=<rot 90 0 0>}]
 dyn_l_string_03 [{shapeType=Capsule, length=2.1455, radius=0.8197, center=<1.0622 -0.086 -0.0002>, rotation=<rot 90 0 0>}]
 dyn_l_string_MSTR [{shapeType=Capsule, length=0.505, radius=0.505}]
 dyn_r_ear [{shapeType=Capsule, length=3.184, radius=3.1828, center=<1.2116 0 0.1695>}]
 dyn_r_ear_MSTR [{shapeType=Capsule, length=0.505, radius=0.505}]
 dyn_r_string_01 [{shapeType=Capsule, length=3.4493, radius=0.5507, center=<2.1628 0 0>, rotation=<rot 90 0 0>}]
 dyn_r_string_02 [{shapeType=Capsule, length=3.0934, radius=0.9066, center=<2.2452 0.0001 0>, rotation=<rot 90 0 0>}]
 dyn_r_string_03 [{shapeType=Capsule, length=2.1455, radius=0.8197, center=<1.7203 0 0>, rotation=<rot 90 0 0>}]
 dyn_r_string_MSTR [{shapeType=Capsule, length=0.505, radius=0.505}]
 spine_04 [{shapeType=Capsule, length=20.0295, radius=12.935}]
 spine_05 [
  {shapeType=Capsule, length=22.1104, radius=10.8541}
  {shapeType=Capsule, length=16.2631, radius=2.2469, center=<0.2948 10.3641 5.1214>, rotation=<rot 90 -3.5763 10.1358>}
  {shapeType=Capsule, length=16.2631, radius=2.2469, center=<0.2624 10.571 -5.678>, rotation=<rot 90 180 -166.2877>}
 ]
}
constraints [
 {
  bone1Name=spine_05
  bone2Name=spine_04
  swing1LimitDegrees=45
  swing1Motion=Free
  swing2LimitDegrees=45
  swing2Motion=Free
  twistLimitDegrees=45
  twistMotion=Free
 }
 {bone1Name=dyn_hood_r, bone2Name=dyn_hood_r_base, swing1LimitDegrees=10, swing2LimitDegrees=10, twistLimitDegrees=10}
 {bone1Name=dyn_hood_front, bone2Name=dyn_hood_front_base, swing1LimitDegrees=10, swing2LimitDegrees=10, twistLimitDegrees=10}
 {bone1Name=dyn_hood_l, bone2Name=dyn_hood_l_base, swing1LimitDegrees=10, swing2LimitDegrees=10, twistLimitDegrees=10}
 {bone1Name=dyn_hood_top, bone2Name=dyn_hood_top_base, swing1LimitDegrees=10, swing2LimitDegrees=10, twistLimitDegrees=10}
 {bone1Name=dyn_r_string_01, bone2Name=dyn_r_string_MSTR, swing1LimitDegrees=45, swing2LimitDegrees=45, twistLimitDegrees=20}
 {bone1Name=dyn_r_string_02, bone2Name=dyn_r_string_01, swing1LimitDegrees=45, swing2LimitDegrees=45, twistLimitDegrees=20}
 {bone1Name=dyn_r_string_03, bone2Name=dyn_r_string_02, swing1LimitDegrees=45, swing2LimitDegrees=45, twistLimitDegrees=20}
 {bone1Name=dyn_l_string_01, bone2Name=dyn_l_string_MSTR, swing1LimitDegrees=45, swing2LimitDegrees=45, twistLimitDegrees=20}
 {bone1Name=dyn_l_string_02, bone2Name=dyn_l_string_01, swing1LimitDegrees=45, swing2LimitDegrees=45, twistLimitDegrees=20}
 {bone1Name=dyn_l_string_03, bone2Name=dyn_l_string_02, swing1LimitDegrees=45, swing2LimitDegrees=45, twistLimitDegrees=20}
 {
  bone1Name=dyn_r_string_MSTR
  bone2Name=spine_05
  swing1LimitDegrees=45
  swing1Motion=Free
  swing2LimitDegrees=45
  swing2Motion=Free
  twistLimitDegrees=45
  twistMotion=Free
 }
 {
  bone1Name=dyn_l_string_MSTR
  bone2Name=spine_05
  swing1LimitDegrees=45
  swing1Motion=Free
  swing2LimitDegrees=45
  swing2Motion=Free
  twistLimitDegrees=45
  twistMotion=Free
 }
 {bone1Name=dyn_l_ear, bone2Name=dyn_l_ear_MSTR, swing1LimitDegrees=30, swing2LimitDegrees=30, twistLimitDegrees=10}
 {bone1Name=dyn_r_ear, bone2Name=dyn_r_ear_MSTR, swing1LimitDegrees=30, swing2LimitDegrees=30, twistLimitDegrees=10}
 {
  bone1Name=dyn_r_ear_MSTR
  bone2Name=spine_05
  swing1LimitDegrees=45
  swing1Motion=Free
  swing2LimitDegrees=45
  swing2Motion=Free
  twistLimitDegrees=45
  twistMotion=Free
 }
 {
  bone1Name=dyn_l_ear_MSTR
  bone2Name=spine_05
  swing1LimitDegrees=45
  swing1Motion=Free
  swing2LimitDegrees=45
  swing2Motion=Free
  twistLimitDegrees=45
  twistMotion=Free
 }
 {
  bone1Name=dyn_hood_r_base
  bone2Name=spine_05
  swing1LimitDegrees=45
  swing1Motion=Free
  swing2LimitDegrees=45
  swing2Motion=Free
  twistLimitDegrees=45
  twistMotion=Free
 }
 {
  bone1Name=dyn_hood_back_base
  bone2Name=spine_05
  swing1LimitDegrees=45
  swing1Motion=Free
  swing2LimitDegrees=45
  swing2Motion=Free
  twistLimitDegrees=45
  twistMotion=Free
 }
 {
  bone1Name=dyn_hood_front_base
  bone2Name=spine_05
  swing1LimitDegrees=45
  swing1Motion=Free
  swing2LimitDegrees=45
  swing2Motion=Free
  twistLimitDegrees=45
  twistMotion=Free
 }
 {
  bone1Name=dyn_hood_l_base
  bone2Name=spine_05
  swing1LimitDegrees=45
  swing1Motion=Free
  swing2LimitDegrees=45
  swing2Motion=Free
  twistLimitDegrees=45
  twistMotion=Free
 }
 {
  bone1Name=dyn_hood_top_base
  bone2Name=spine_05
  swing1LimitDegrees=45
  swing1Motion=Free
  swing2LimitDegrees=45
  swing2Motion=Free
  twistLimitDegrees=45
  twistMotion=Free
 }
]
